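# Sound cue for the shopping cart vehicle's destruction, reusing shopping cart demolition impact waves.
# A quality-level node selects one of four pitch-modulated branches, shrinking variation from a random pick of three waves to two, then to a single wave in the last two.
# It plays in the impacts sound class with shopping cart movement attenuation and a shared object-impact concurrency limit.
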SoundCue /Game/Athena/Sounds/Vehicles/ShoppingCart/ShoppingCart_Destroy_Cue
root=QualityLevel_1
nodes [
 {id=Modulator_5, children [Random_4], properties {pitchMax=1.2, pitchMin=0.92}}
 {id=Modulator_6, children [Random_5], properties {pitchMax=1.2, pitchMin=0.92}}
 {id=Modulator_8, children [WavePlayer_16], properties {pitchMax=1.2, pitchMin=0.92}}
 {id=Modulator_9, children [WavePlayer_17], properties {pitchMax=1.2, pitchMin=0.92}}
 {id=QualityLevel_1, children [Modulator_5,Modulator_6,Modulator_8,Modulator_9]}
 {id=Random_4, children [WavePlayer_5, WavePlayer_12, WavePlayer_13], properties {weights [1,1,1]}}
 {id=Random_5, children [WavePlayer_14, WavePlayer_15], properties {weights [1,1]}}
 {id=WavePlayer_12, properties {soundWaveAssetPtr=/Game/Sounds/Fort_Impact_Sounds/Demolition/ShoppingCart/shoppingcart_impact_02}}
 {id=WavePlayer_13, properties {soundWaveAssetPtr=/Game/Sounds/Fort_Impact_Sounds/Demolition/ShoppingCart/shoppingcart_impact_03}}
 {id=WavePlayer_14, properties {soundWaveAssetPtr=/Game/Sounds/Fort_Impact_Sounds/Demolition/ShoppingCart/shoppingcart_impact_01}}
 {id=WavePlayer_15, properties {soundWaveAssetPtr=&shoppingcart_impact_02}}
 {id=WavePlayer_16, properties {soundWaveAssetPtr=&shoppingcart_impact_01}}
 {id=WavePlayer_17, properties {soundWaveAssetPtr=&shoppingcart_impact_01}}
 {id=WavePlayer_5, properties {soundWaveAssetPtr=&shoppingcart_impact_01}}
]
properties {
 attenuationSettings=/Game/Sounds/Attenuation/Fort_ShoppingCart_Movement_Att
 maxDistance=5300
 soundClassObject=&Fort_Impacts
 volumeMultiplier=0.6
 concurrencyOverrides {resolutionRule=PreventNew}
 concurrencySet [&CONC_Object_Impacts_FourMax]
}
