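# Sound cue for the shopping cart vehicle, played by name as the driver equip sound.
# A single wave player feeds the output directly with a temp-asset cart grab foley wave, with no randomization or modulation.
# The cue plays at reduced volume in the vehicles sound class, spatialized through the shared shopping cart movement attenuation.
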SoundCue /Game/Athena/Sounds/Vehicles/ShoppingCart/ShoppingCart_Driver_Equip_Cue
root=WavePlayer_2
nodes [{id=WavePlayer_2, properties {soundWaveAssetPtr=/Game/Athena/Sounds/Vehicles/ShoppingCart/TempAssets/Cart_Foley_Grab_04b}}]
properties {
 attenuationSettings=/Game/Sounds/Attenuation/Fort_ShoppingCart_Movement_Att
 maxDistance=5300
 soundClassObject=&Fort_Vehicles
 volumeMultiplier=0.3
}
