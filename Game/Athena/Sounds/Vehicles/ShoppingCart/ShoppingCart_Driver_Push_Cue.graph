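# Sound cue for the driver pushing a shopping cart vehicle.
# A mixer layers one of three randomly chosen cart push waves with a quieter Commando landing foley sound.
# The foley layer is slightly delayed and modulated to a raised, varied pitch and random volume, likely adding a footfall to each push.
# The cue plays at reduced volume in the vehicles sound class with shopping cart movement attenuation.
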SoundCue /Game/Athena/Sounds/Vehicles/ShoppingCart/ShoppingCart_Driver_Push_Cue
root=Mixer_0
nodes [
 {id=Delay_0, children [Modulator_5], properties {delayMax=0.1, delayMin=0.1}}
 {id=Mixer_0, children [Random_1,Delay_0], properties {inputVolume [1,0.3]}}
 {id=Modulator_5, children [WavePlayer_3], properties {pitchMax=1.5, pitchMin=1.3, volumeMax=1.3, volumeMin=0.7}}
 {id=Random_1, children [WavePlayer_5,WavePlayer_6,WavePlayer_7], properties {weights [1,1,1]}}
 {id=WavePlayer_3, properties {soundWaveAssetPtr=/Game/Sounds/Fortress_Foley/Commando/Commando_Land_002}}
 {id=WavePlayer_5, properties {soundWaveAssetPtr=&/Shopping_Cart_Push_01}}
 {id=WavePlayer_6, properties {soundWaveAssetPtr=&/Shopping_Cart_Push_02}}
 {id=WavePlayer_7, properties {soundWaveAssetPtr=&/Shopping_Cart_Push_03}}
]
properties {
 attenuationSettings=/Game/Sounds/Attenuation/Fort_ShoppingCart_Movement_Att
 maxDistance=5300
 soundClassObject=&Fort_Vehicles
 volumeMultiplier=0.4
}
