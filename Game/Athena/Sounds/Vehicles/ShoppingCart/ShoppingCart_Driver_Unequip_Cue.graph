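# Sound cue for the shopping cart driver unequip, playing a cart foley grab sound from a TempAssets folder.
# A quality level node picks the standard wave for the first level and the "_opt" variant for the other three.
# Playback runs at reduced volume through the vehicles sound class with shopping cart movement attenuation.
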SoundCue /Game/Athena/Sounds/Vehicles/ShoppingCart/ShoppingCart_Driver_Unequip_Cue
root=QualityLevel_0
nodes [
 {id=QualityLevel_0, children [WavePlayer_1,WavePlayer_0,WavePlayer_0,WavePlayer_0]}
 {id=WavePlayer_0, properties {soundWaveAssetPtr=/Game/Athena/Sounds/Vehicles/ShoppingCart/TempAssets/Cart_Foley_Grab_01_opt}}
 {id=WavePlayer_1, properties {soundWaveAssetPtr=/Game/Athena/Sounds/Vehicles/ShoppingCart/TempAssets/Cart_Foley_Grab_01}}
]
properties {
 attenuationSettings=/Game/Sounds/Attenuation/Fort_ShoppingCart_Movement_Att
 maxDistance=5300
 soundClassObject=&Fort_Vehicles
 volumeMultiplier=0.4
}
