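# Sound cue for the Athena shopping cart vehicle's looping movement audio, layering rolling, coasting, skid, in-air wheel spin, scrape, wind and water loops.
# A mixer sums the layers, each volume driven by a continuous modulator on a named runtime parameter.
# The rolling layer crossfades light to medium loops by Speed.
# The water wake nests two modulators, so Water and Speed both gate its volume while Speed also shifts pitch.
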SoundCue /Game/Athena/Sounds/Vehicles/ShoppingCart/ShoppingCart_Movement_Loop_Cue
root=Mixer_0
nodes [
 {
  id=Mixer_0
  children [
   ModulatorContinuous_1
   ModulatorContinuous_2
   ModulatorContinuous_3
   ModulatorContinuous_4
   ModulatorContinuous_5
   ModulatorContinuous_6
   ModulatorContinuous_7
  ]
  properties {inputVolume [2,0.4,0.3,0.4,0.2,1,1]}
 }
 {id=ModulatorContinuous_0, children [WavePlayer_7], properties {volumeModulationParams {default=0, parameterName=Water}}}
 {id=ModulatorContinuous_1, children [ParamCrossFade_1], properties {volumeModulationParams {default=0, parameterName=Speed}}}
 {id=ModulatorContinuous_2, children [WavePlayer_1], properties {volumeModulationParams {default=0, parameterName=Coast}}}
 {id=ModulatorContinuous_3, children [WavePlayer_2], properties {volumeModulationParams {default=0, parameterName=Skid}}}
 {id=ModulatorContinuous_4, children [WavePlayer_4], properties {volumeModulationParams {default=0, parameterName=InAir}}}
 {id=ModulatorContinuous_5, children [WavePlayer_5], properties {volumeModulationParams {default=0, parameterName=Scrape}}}
 {id=ModulatorContinuous_6, children [WavePlayer_6], properties {volumeModulationParams {default=0, parameterName=Wind}}}
 {
  id=ModulatorContinuous_7
  children [ModulatorContinuous_0]
  properties {pitchModulationParams {default=0, minOutput=0.8, parameterName=Speed}, volumeModulationParams {default=0, maxInput=0.2, parameterName=Speed}}
 }
 {
  id=ParamCrossFade_1
  children [WavePlayer_0,WavePlayer_3]
  properties {
   paramName=Speed
   crossFadeInput [
    {fadeInDistanceEnd=0, fadeInDistanceStart=0, fadeOutDistanceEnd=0.7, fadeOutDistanceStart=0.3, volume=0.7}
    {fadeInDistanceEnd=0.7, fadeInDistanceStart=0.3, fadeOutDistanceEnd=1.1, fadeOutDistanceStart=1.1, volume=0.3}
   ]
  }
 }
 {id=WavePlayer_0, properties {bLooping=true, soundWaveAssetPtr=/Game/Athena/Sounds/Vehicles/ShoppingCart/TempAssets/Cart_Loop_Light_01b}}
 {id=WavePlayer_1, properties {bLooping=true, soundWaveAssetPtr=&/ShoppingCart_Coast_02}}
 {id=WavePlayer_2, properties {bLooping=true, soundWaveAssetPtr=/Game/Athena/Sounds/Vehicles/ShoppingCart/TempAssets/Cart_Loop_Skid_01}}
 {id=WavePlayer_3, properties {bLooping=true, soundWaveAssetPtr=/Game/Athena/Sounds/Vehicles/ShoppingCart/TempAssets/Cart_Loop_Med_02}}
 {id=WavePlayer_4, properties {bLooping=true, soundWaveAssetPtr=&/ShoppingCart_WheelSpin_01}}
 {id=WavePlayer_5, properties {bLooping=true, soundWaveAssetPtr=/Game/Athena/Sounds/Vehicles/ShoppingCart/TempAssets/Cart_Loop_Drag_Heavy_01}}
 {id=WavePlayer_6, properties {bLooping=true, soundWaveAssetPtr=/Game/Athena/Sounds/Wind_Falling/Wind_UmbrellaClosed_Loop_01}}
 {id=WavePlayer_7, properties {bLooping=true, soundWaveAssetPtr=/Game/Athena/Sounds/Water/Interaction/Ski_Water_Player_Wake_Constant_Loop_01}}
]
properties {
 attenuationSettings=/Game/Sounds/Attenuation/Fort_ShoppingCart_Movement_Att
 maxDistance=5300
 soundClassObject=&Fort_Vehicles
 volumeMultiplier=1
}
