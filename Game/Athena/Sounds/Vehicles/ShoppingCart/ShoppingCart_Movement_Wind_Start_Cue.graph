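# Sound cue for the wind start sound of shopping cart movement, reusing parachute skydive loop-enter waves from the falling wind set.
# A quality-level node at the root sends the first two levels to an evenly weighted random pick of three waves.
# The other two levels share a single optimized variant of the third wave, dropping the variation.
# It plays in the vehicles sound class with shopping cart movement attenuation.
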SoundCue /Game/Athena/Sounds/Vehicles/ShoppingCart/ShoppingCart_Movement_Wind_Start_Cue
root=QualityLevel_0
nodes [
 {id=QualityLevel_0, children [Random_0,Random_0,WavePlayer_1,WavePlayer_1]}
 {id=Random_0, children [WavePlayer_0,WavePlayer_3,WavePlayer_4], properties {weights [1,1,1]}}
 {id=WavePlayer_0, properties {soundWaveAssetPtr=/Game/Athena/Sounds/Wind_Falling/sfx_parachute_skydive_loop_enter_03}}
 {id=WavePlayer_1, properties {soundWaveAssetPtr=/Game/Athena/Sounds/Wind_Falling/sfx_parachute_skydive_loop_enter_03_opt}}
 {id=WavePlayer_3, properties {soundWaveAssetPtr=/Game/Athena/Sounds/Wind_Falling/sfx_parachute_skydive_loop_enter_01}}
 {id=WavePlayer_4, properties {soundWaveAssetPtr=/Game/Athena/Sounds/Wind_Falling/sfx_parachute_skydive_loop_enter_02}}
]
properties {
 attenuationSettings=/Game/Sounds/Attenuation/Fort_ShoppingCart_Movement_Att
 maxDistance=5300
 soundClassObject=&Fort_Vehicles
 volumeMultiplier=1
}
