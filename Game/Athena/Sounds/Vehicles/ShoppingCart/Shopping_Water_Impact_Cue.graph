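# Sound cue for the shopping cart vehicle's water impact, reusing the Meatball vehicle's light water impact waves.
# A Random node picks among seven equally weighted wave variations and feeds a Modulator at the root.
# The cue plays at half volume with raised pitch, routed through the vehicles sound class and the shopping cart movement attenuation.
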SoundCue /Game/Athena/Sounds/Vehicles/ShoppingCart/Shopping_Water_Impact_Cue
root=Modulator_0
nodes [
 {id=Modulator_0, children [Random_0]}
 {
  id=Random_0
  children [WavePlayer_8, WavePlayer_9, WavePlayer_10, WavePlayer_11, WavePlayer_12, WavePlayer_13, WavePlayer_14]
  properties {weights [1,1,1,1,1,1,1]}
 }
 {id=WavePlayer_10, properties {soundWaveAssetPtr=/Game/Athena/Sounds/Vehicles/Meatball/SoundWaves/Veh_Meatball_Water_Impact_Lt_03}}
 {id=WavePlayer_11, properties {soundWaveAssetPtr=/Game/Athena/Sounds/Vehicles/Meatball/SoundWaves/Veh_Meatball_Water_Impact_Lt_04}}
 {id=WavePlayer_12, properties {soundWaveAssetPtr=/Game/Athena/Sounds/Vehicles/Meatball/SoundWaves/Veh_Meatball_Water_Impact_Lt_05}}
 {id=WavePlayer_13, properties {soundWaveAssetPtr=/Game/Athena/Sounds/Vehicles/Meatball/SoundWaves/Veh_Meatball_Water_Impact_Lt_06}}
 {id=WavePlayer_14, properties {soundWaveAssetPtr=/Game/Athena/Sounds/Vehicles/Meatball/SoundWaves/Veh_Meatball_Water_Impact_Lt_07}}
 {id=WavePlayer_8, properties {soundWaveAssetPtr=/Game/Athena/Sounds/Vehicles/Meatball/SoundWaves/Veh_Meatball_Water_Impact_Lt_01}}
 {id=WavePlayer_9, properties {soundWaveAssetPtr=/Game/Athena/Sounds/Vehicles/Meatball/SoundWaves/Veh_Meatball_Water_Impact_Lt_02}}
]
properties {
 attenuationSettings=/Game/Sounds/Attenuation/Fort_ShoppingCart_Movement_Att
 maxDistance=5300
 pitchMultiplier=1.5
 soundClassObject=&Fort_Vehicles
 volumeMultiplier=0.5
}
